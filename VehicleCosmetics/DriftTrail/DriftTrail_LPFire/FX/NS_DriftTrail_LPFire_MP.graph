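# Niagara system for the LPFire drift trail vehicle cosmetic, drawing effects at the car's wheels.
# Four emitters layer short-lived sub-UV dots, core meshes with sprites, twin ribbons and wheel-wrap meshes, positioned from the vehicle skeletal mesh.
# Each emitter picks colours through painted-variant overrides and masks particles per wheel with FN_DisableTireFX.
# A front-end preview flag retunes velocity, brightness and spawning, including preview-only sprite spawning in the core emitter.
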NiagaraSystem /VehicleCosmetics/DriftTrail/DriftTrail_LPFire/FX/NS_DriftTrail_LPFire_MP
dependencies {
 usedDataInterfaces [
  &NiagaraDataInterfaceCurve
  &NiagaraDataInterfaceSpriteRendererInfo
  &NiagaraDataInterfaceSkeletalMesh
 ]
 usedDynamicInputs []
 usedModules [
  &self:SystemSpawnScript.SystemScriptSource.SystemScriptGraph.NiagaraNodeAssignment_0.SetVariables_#1
  &SystemState
  &FN_UnpackVehicleFlags
  /Game/Effects/Niagara/Modules/Vehicle/FN_UpdateVehicleAttribs
  /Game/Effects/Niagara/Modules/Vehicle/FN_VehicleFlags_Cheatcodes
 ]
}
emitters {
 Dots {
  emitterUpdateScript [
   &EmitterState(Life Cycle Mode=::NewEnumerator1, Loop Duration=1, UseLoopDelay=false)
   &SpawnRate(
    SpawnRate=/VehicleCosmetics/FX/Niagara/Modules/VC_Mask_Floats(
     FrontEnd Value:NiagaraFloat=&RandomRangeFloat(Minimum=80, Maximum=120)
     blsFrontEndPreview:NiagaraBool=$User.bIsFrontendPreview
     SpeedMin Value:NiagaraFloat=&RandomRangeFloat(Minimum=80, Maximum=120)
     SpeedMax Value:NiagaraFloat=&RandomRangeFloat(Minimum=80, Maximum=120)
     Speed Alpha:NiagaraFloat=$System.Speed_Alpha
    )
    Use Spawn Probability=false
   )
  ]
  particleSpawnScript [
   &VC_PaintedVariantsOverride(
    Black=<rgba 0.02 0.02 0.02 1>
    BurntSienna=<rgba 0.14 0.0245 0.0014 1>
    Cobalt=<rgba 0.0089 0.0676 0.8896 1>
    Crimson=<rgba 1 0 0.0167 2>
    DefaultColor=<rgba 0 0.4 0.3667 1>
    ForestGreen=<rgba 0.1225 0.7 0.07 1>
    Grey=<rgba 0.1 0.1 0.1 1>
    Lime=<rgba 0.25 1 0 1>
    Orange=<rgba 1 0.1333 0 1>
    Pink=<rgba 1 0.2381 0.7143 1>
    Purple=<rgba 0.3 0 0.9 1>
    R1 Black:LinearColor=<rgba 0.003 0.003 0.003 1>
    R1 BurntSienna:LinearColor=<rgba 1 0.4 0.1 1>
    R1 Cobalt:LinearColor=<rgba 0 0.0169 0.913 1>
    R1 Crimson:LinearColor=<rgba 0.3 0 0.0081 1>
    R1 DefaultColor=<rgba 0.0771 0.4 0.3731 1>
    R1 ForestGreen:LinearColor=<rgba 0.238 0.8958 0.2712 1>
    R1 Lime:LinearColor=<rgba 0.4932 0.9375 0.0488 1>
    R1 Orange:LinearColor=<rgba 0.9115 0.3179 0 1>
    R1 Pink:LinearColor=<rgba 5 0.599 3.3996 1>
    R1 Saffron:LinearColor=<rgba 5 3.9488 0.4948 1>
    R1 TitaniumWhite:LinearColor=<rgba 100 100 100 1>
    Saffron=<rgba 0.45 0.375 0 1>
    SkyBlue=<rgba 0.1 0.34 1 1>
    TitaniumWhite=<rgba 0.3 0.3 0.3 1>
    Use R1 Painted Overrides=true
   )
   &InitializeParticle(
    Color Mode=::NewEnumerator0
    Lifetime Max=0.15
    Lifetime Min=0.09
    Lifetime Mode=::NewEnumerator1
    Sprite Rotation Angle=&VectorLength(Vector=<0 1 0>)
    Sprite Rotation Mode=::NewEnumerator0
    Sprite Size Mode=::NewEnumerator1
    Uniform Sprite Size Max=15
    Uniform Sprite Size Min=12
    UsePositionOffset=false
    Write Material Random=true
   )
   &PartitionParticles()
   &SkeletalMeshLocation(
    Bone / Socket Index=&Add_Integer(A=$Particles.PartitionParticles.Partition, B=4)
    Bone / Socket Sampling Mode=::NewEnumerator5
    Interpolate Mesh Sampling=false
    MaskSpawnGroup=false
    Mesh Sampling Type=::NewEnumerator6
    Sampled Position Offset=&AddVectorToPosition(Position=<-30 0 0>, Vector=$User.offset)
    Skeletal Mesh=$System.SkeletalMesh
   )
   &AddVelocity(
    Velocity=&VC_Mask_Vectors(
     FrontEnd Value:Vector3f=&RandomRangeVector(Minimum=<500 0 0>, Maximum=<1000 0 0>)
     blsFrontEndPreview:NiagaraBool=$User.bIsFrontendPreview
     SpeedMin Value:Vector3f=&RandomRangeVector(Minimum=<500 0 0>, Maximum=<1000 0 0>)
     SpeedMax Value:Vector3f=&RandomRangeVector(Minimum=<500 0 0>, Maximum=<100 0 0>)
     Speed Alpha:NiagaraFloat=$System.Speed_Alpha
    )
    Velocity Speed Scale=&MakeCustomFloatFromBool(Bool=$User.bIsFrontendPreview, True Float=-1.8, False Float=-2.4)
   )
   &ShapeLocation(
    MaskSpawnGroup=false
    Override Seed=false
    Sphere Radius=&RandomRangeFloat(Minimum=12, Maximum=18)
    UseOwnerScale=true
   )
  ]
  particleUpdateScript [
   &FN_DisableTireFX(
    Mask Index Attribute:NiagaraInt32=$Particles.PartitionParticles.Partition
    Mask Index FL=2
    Mask Index FR=3
    Mask Index RL=0
    Mask Index RR=1
   )
   &ParticleState()
   &InheritSourceMovement(
    Applied Position Scale=&VectorFromFloat(Value=&FloatFromCurve(FloatCurve=DataInterface() {propertyValues {Curve {keys [<at 0.5 1 RCIM_Cubic>, {arriveTangent=-1.2348, interpMode=RCIM_Cubic, leaveTangent=-1.2348, tangentMode=RCTM_User, time=1, value=0.8}]}}}))
    ApplyForce=false
    ApplyPosition=true
    ApplyPositionFalloff=false
   )
   &ScaleSpriteSize(Uniform Curve Sprite Scale=DataInterface() {propertyValues {Curve {keys [
        {tangentMode=RCTM_User, time=0, value=1}
        {arriveTangent=-0.0405, interpMode=RCIM_Cubic, leaveTangent=-0.0405, tangentMode=RCTM_User, time=0.5, value=1}
        {arriveTangent=-3.7185, interpMode=RCIM_Cubic, leaveTangent=-3.7185, tangentMode=RCTM_User, time=1, value=0}
       ]}}})
   &SubUVAnimation(
    Random Start Frame=true
    "Sprite Renderer "=DataInterface() {propertyValues {SpriteRenderer=&self:Dots_2.NiagaraSpriteRendererProperties_4}}
    UseEndFrame=false
    UseStartFrame=false
   )
   &SolveForcesAndVelocity(Clamp Velocity=false, Limit Acceleration=false)
   /SetVariables(Particles.R2Color:LinearColor=&MultiplyLinearColor(
     Linear Color=$Particles.R1 Color
     Scale Factor=&MakeVector4(
      X=10
      Y=10
      Z=10
      W=&MakeCustomFloatFromBool(
       Bool=$User.bIsFrontendPreview
       True Float=&FloatFromCurve(FloatCurve=DataInterface() {propertyValues {Curve {keys [
            {arriveTangent=0.0266, interpMode=RCIM_Cubic, leaveTangent=0.0266, tangentMode=RCTM_User, time=0, value=0}
            {arriveTangent=1.1179, interpMode=RCIM_Cubic, leaveTangent=1.1179, tangentMode=RCTM_User, time=0.25, value=0.6228}
            {arriveTangent=0.0224, interpMode=RCIM_Cubic, leaveTangent=0.0224, tangentMode=RCTM_User, time=1, value=1}
           ]}}})
       False Float=&FloatFromCurve(FloatCurve=DataInterface() {propertyValues {Curve {keys [
            {arriveTangent=9.0725, interpMode=RCIM_Cubic, leaveTangent=9.0725, tangentMode=RCTM_User, time=0, value=0}
            {arriveTangent=-2.5249, interpMode=RCIM_Cubic, leaveTangent=-2.5249, tangentMode=RCTM_User, time=0.1408, value=0.9847}
            {arriveTangent=-0.0405, interpMode=RCIM_Cubic, leaveTangent=-0.0405, tangentMode=RCTM_User, time=1, value=0}
           ]}}})
      )
     )
    )) {isSetParameters=true}
   &ScaleColor(
    Scale Alpha=&MakeCustomFloatFromBool(
     Bool=$User.bIsFrontendPreview
     True Float=&FloatFromCurve(FloatCurve=DataInterface() {propertyValues {Curve {keys [
          {arriveTangent=0.0266, interpMode=RCIM_Cubic, leaveTangent=0.0266, tangentMode=RCTM_User, time=0, value=0}
          {arriveTangent=1.1179, interpMode=RCIM_Cubic, leaveTangent=1.1179, tangentMode=RCTM_User, time=0.25, value=0.6228}
          {arriveTangent=0.0224, interpMode=RCIM_Cubic, leaveTangent=0.0224, tangentMode=RCTM_User, time=1, value=1}
         ]}}})
     False Float=&FloatFromCurve(FloatCurve=DataInterface() {propertyValues {Curve {keys [
          {arriveTangent=9.0725, interpMode=RCIM_Cubic, leaveTangent=9.0725, tangentMode=RCTM_User, time=0, value=0}
          {arriveTangent=-2.5249, interpMode=RCIM_Cubic, leaveTangent=-2.5249, tangentMode=RCTM_User, time=0.1408, value=0.9847}
          {arriveTangent=-0.0405, interpMode=RCIM_Cubic, leaveTangent=-0.0405, tangentMode=RCTM_User, time=1, value=0}
         ]}}})
    )
    Scale RGB=&Multiply_VectorByFloat(
     Vector=<1 1 1>
     Float=&MakeCustomFloatFromBool(
      Bool=$User.bIsFrontendPreview
      True Float=0.5
      False Float=&MakeCustomFloatFromBool(Bool=$User.bIsPainted, True Float=0.3, False Float=0.7)
     )
    )
    ScaleA=true
    ScaleRGB=true
   ) {module=ScaleColor001}
  ]
  rendererClasses [&NiagaraSpriteRendererProperties]
  renderers [
   {rendererClass=&NiagaraSpriteRendererProperties}
   {rendererClass=&NiagaraSpriteRendererProperties, rendererIndex=1}
  ]
 }
 NE_VC_Core_Outer {
  emitterUpdateScript [
   &EmitterState(
    Inactive Response=::NewEnumerator1
    Life Cycle Mode=::NewEnumerator1
    Loop Behavior=::NewEnumerator1
    Loop Duration Mode=::NewEnumerator1
    UseLoopDelay=false
   )
   &SpawnRate(
    SpawnRate=&MakeCustomFloatFromBool(Bool=$User.bIsFrontendPreview, True Float=10, False Float=0)
    Use Spawn Probability=false
   ) {enabled=false, module=SpawnRate001}
   &SpawnBurst_Instantaneous(Spawn Count=4, Spawn Group=1, Use Spawn Probability=false, UseLoopCountLimit=false) {module=SpawnBurst_Instantaneous001}
   &SpawnBurst_Instantaneous(Spawn Count=4, Use Spawn Probability=false, UseLoopCountLimit=false)
   &SpawnRate(
    SpawnGroup=2
    SpawnRate=&MakeCustomFloatFromBool(Bool=$User.bIsFrontendPreview, True Float=20, False Float=0)
    Use Spawn Probability=false
   )
  ]
  particleSpawnScript [
   &VC_PaintedVariantsOverride(
    Black=<rgba 0.007 0.007 0.007 1>
    BurntSienna=<rgba 0.1146 0.0191 0 1>
    Cobalt=<rgba 0.012 0.0314 0.4 1>
    Crimson=<rgba 0.35 0 0.0095 1>
    DefaultColor=<rgba 0.025 0.2125 0.25 1>
    ForestGreen=<rgba 0.0141 0.47 0.0293 1>
    Grey=<rgba 0.055 0.055 0.055 1>
    Lime=<rgba 0.1048 0.599 0.006 1>
    Orange=&MakeCustomLinearColorFromBool(Bool=$User.bIsFrontendPreview, True Color=<rgba 0.4635 0.0464 0 1>, False Color=<rgba 0.8 0.08 0 1>)
    Pink=<rgba 0.5 0.119 0.3571 1>
    Purple=<rgba 0.1456 0.0068 0.34 1>
    R1 Black:LinearColor=<rgba 0.006 0.006 0.006 1>
    R1 BurntSienna:LinearColor=&MakeCustomLinearColorFromBool(Bool=$User.bIsFrontendPreview, True Color=<rgba 2.5 0.4375 0.025 1>, False Color=<rgba 10 1.75 0.1 1>)
    R1 Cobalt:LinearColor=<rgba 0 0.02 0.4 1>
    R1 Crimson:LinearColor=<rgba 0.21 0.0023 0.0061 1>
    R1 DefaultColor=<rgba 0 0.1833 0.2 1>
    R1 ForestGreen:LinearColor=&MakeCustomLinearColorFromBool(Bool=$User.bIsFrontendPreview, True Color=<rgba 0 0.35 0.0117 1>, False Color=<rgba 0.0467 0.4 0 1>)
    R1 Grey:LinearColor=<rgba 0.07 0.07 0.07 1>
    R1 Lime:LinearColor=<rgba 0.1875 0.25 0 1>
    R1 Orange:LinearColor=<rgba 0.85 0.0708 0 1>
    R1 Pink:LinearColor=&MakeCustomLinearColorFromBool(Bool=$User.bIsFrontendPreview, True Color=<rgba 10 2.381 7.1428 1>, False Color=<rgba 10 2.381 7.1428 1>)
    R1 Purple:LinearColor=<rgba 0.05 0 0.15 1>
    R1 Saffron:LinearColor=<rgba 1.5 0.5 0 1>
    R1 SkyBlue:LinearColor=<rgba 0 0.15 0.9 1>
    R1 TitaniumWhite:LinearColor=<rgba 4 4 4 1>
    R2 Black:LinearColor=<rgba 0.006 0.006 0.006 1>
    R2 BurntSienna:LinearColor=<rgba 2.5 0.4375 0.025 1>
    R2 Cobalt:LinearColor=<rgba 0 0.035 0.7 1>
    R2 Crimson:LinearColor=<rgba 0.4635 0 0.0126 1>
    R2 DefaultColor:LinearColor=<rgba 0 0.2 0.1833 1>
    R2 ForestGreen:LinearColor=<rgba 0 0.55 0.0183 1>
    R2 Grey:LinearColor=<rgba 0.05 0.05 0.05 1>
    R2 Lime:LinearColor=<rgba 0.3 0.4 0 1>
    R2 Orange:LinearColor=<rgba 1.2 0.1 0 1>
    R2 Pink:LinearColor=&MakeCustomLinearColorFromBool(Bool=$User.bIsFrontendPreview, True Color=<rgba 10 2.381 7.1428 1>, False Color=<rgba 2 0.4762 1.4286 1>)
    R2 Purple:LinearColor=<rgba 0.0807 0 0.22 1>
    R2 Saffron:LinearColor=<rgba 0.7 0.5367 0 1>
    R2 SkyBlue:LinearColor=<rgba 0 0.15 0.9 1>
    R2 TitaniumWhite:LinearColor=<rgba 4 4 4 1>
    Saffron=<rgba 0.2 0.1842 0.01 1>
    SkyBlue=<rgba 0.051 0.1955 0.34 1>
    TitaniumWhite=<rgba 0.2188 0.2188 0.2188 1>
    Use R1 Painted Overrides=true
    Use R2 Painted Overrides=true
   )
   /SetVariables(Particles.VisTag:NiagaraInt32=&SpawnGroup()) {isSetParameters=true}
   &Color(Color=&MakeCustomLinearColorFromBool(
     Bool=$User.bIsFrontendPreview
     True Color=&MakeCustomLinearColorFromBool(Bool=$User.bIsPainted, True Color=$User.VariableForPainted, False Color=<rgba 0 0.24 0.24 1>)
     False Color=&MakeCustomLinearColorFromBool(Bool=$User.bIsPainted, True Color=$User.VariableForPainted, False Color=<rgba 0 0.24 0.24 1>)
    )) {enabled=false}
   &InitializeParticle(
    Color Mode=::NewEnumerator0
    Lifetime=/VehicleCosmetics/FX/Niagara/Modules/MergeEmitters/VC_FloatBySpawnGroup(
     SpawnGroup or ExecIndex:NiagaraInt32=0
     CustomInt:NiagaraInt32=0
     MaximumSpawnGroups:NiagaraInt32=3
     Float SG 0:NiagaraFloat=$Emitter.CurrentLoopDuration
     Float SG 1:NiagaraFloat=$Emitter.CurrentLoopDuration
     Float SG 2:NiagaraFloat=&RandomRangeFloat(Minimum=0.1, Maximum=0.2)
     Float SG 3:NiagaraFloat=0
     Float SG 4:NiagaraFloat=0
     Float SG 5:NiagaraFloat=0
    )
    Mesh Scale=<0.6 0.48 0.48>
    Mesh Scale Mode=::NewEnumerator4
    Sprite Size=&RandomRangeVector2D(Minimum=<36 18>, Maximum=<18 6>)
    Sprite Size Mode=::NewEnumerator4
    UsePositionOffset=false
    Write Material Random=true
   )
   &InitialMeshOrientation(Mesh Orientation Mode=::NewEnumerator2, Orientation Vector=<-1 0 0>, Use Rotation Vector=false)
   &ApplyInitialForces()
   &AlignSpriteToMeshOrientation(
    Mesh Orientation Relative Sprite Facing Vector:Vector3f=&MakeCustomVectorFromBool(
     Bool=$User.bIsFrontendPreview
     True Vector=&RandomRangeVector(Minimum=<0 1 0>, Maximum=<0 1 1>, Randomness Mode=::NewEnumerator1)
     False Vector=<0 0 1>
    )
    SetAlignment=true
    SetFacing=true
   ) {module=AlignSpriteToMeshOrientation001}
   &CameraOffset(Camera Offset Amount=&VC_FloatBySpawnGroup(
     SpawnGroup or ExecIndex:NiagaraInt32=0
     CustomInt:NiagaraInt32=0
     MaximumSpawnGroups:NiagaraInt32=2
     Float SG 0:NiagaraFloat=-10
     Float SG 1:NiagaraFloat=20
     Float SG 2:NiagaraFloat=100
     Float SG 3:NiagaraFloat=0
     Float SG 4:NiagaraFloat=0
     Float SG 5:NiagaraFloat=0
    ))
   &PartitionParticles()
   &SkeletalMeshLocation(
    Bone / Socket Index=&Add_Integer(A=$Particles.PartitionParticles.Partition, B=4)
    Bone / Socket Sampling Mode=::NewEnumerator5
    Interpolate Mesh Sampling=false
    MaskSpawnGroup=false
    Mesh Sampling Type=::NewEnumerator6
    Sampled Position Offset=&AddVectorToPosition(
     Position=/VehicleCosmetics/FX/Niagara/Modules/MergeEmitters/VC_VectorBySpawnGroup(
      SpawnGroup or ExecIndex:NiagaraInt32=0
      CustomInt:NiagaraInt32=0
      MaximumSpawnGroups:NiagaraInt32=3
      Vector Group 0:Vector3f=<-0.7 0 0>
      Vector Group 1:Vector3f=<-10 1 1>
      Vector Group 2:Vector3f=<-25 0 0>
      Vector Group 3:Vector3f=<0 0 0>
      Vector Group 4:Vector3f=<0 0 0>
      Vector Group 5:Vector3f=<0 0 0>
     )
     Vector=$User.offset
    )
    Skeletal Mesh=$System.SkeletalMesh
   )
   &AddVelocity(
    Velocity=&VC_VectorBySpawnGroup(
     SpawnGroup or ExecIndex:NiagaraInt32=0
     CustomInt:NiagaraInt32=0
     MaximumSpawnGroups:NiagaraInt32=3
     Vector Group 0:Vector3f=<0 0 0>
     Vector Group 1:Vector3f=<0 0 0>
     Vector Group 2:Vector3f=&RandomRangeVector(Minimum=<150 0 0>, Maximum=<300 0 0>)
     Vector Group 3:Vector3f=<0 0 0>
     Vector Group 4:Vector3f=<0 0 0>
     Vector Group 5:Vector3f=<0 0 0>
    )
    Velocity Speed Scale=-1
   )
   /VehicleCosmetics/FX/Niagara/Modules/MergeEmitters/VC_ExtraBindingColors(
    ExtraColors=1
    R1 Color:LinearColor=&MakeCustomLinearColorFromBool(
     Bool=$User.bIsFrontendPreview
     True Color=&MakeCustomLinearColorFromBool(Bool=$User.bIsPainted, True Color=$User.VariableForPainted2, False Color=<rgba 0 0.1833 0.2 1>)
     False Color=&MakeCustomLinearColorFromBool(Bool=$User.bIsPainted, True Color=$User.VariableForPainted2, False Color=<rgba 0 0.1833 0.2 1>)
    )
   ) {enabled=false}
   /SetVariables(Particles.SPEARSCOLOR:LinearColor=&MakeCustomLinearColorFromBool(Bool=$User.bIsPainted, True Color=$User.VariableForPainted2, False Color=<rgba 0 0.2 0.1833 1>)) {enabled=false, isSetParameters=true}
  ]
  particleUpdateScript [
   /SetVariables() {isSetParameters=true}
   &FN_DisableTireFX(
    Mask Index Attribute:NiagaraInt32=$Particles.PartitionParticles.Partition
    Mask Index FL=2
    Mask Index FR=3
    Mask Index RL=0
    Mask Index RR=1
   )
   &DynamicMaterialParameters(
    Index 0 Param 1=1
    Index 0 Param 2=1
    Index 0 Param 3=1
    Module.Index0Param0Write=true
    Module.Index0Param1Write=true
    Module.Index0Param2Write=true
    Module.Index0Param3Write=true
   ) {enabled=false}
   /SetVariables(
    Particles.R1 Color Scale:LinearColor=&VC_DynamicScaleColor(
     Color to Scale:LinearColor=$Particles.Initial.R1 Color
     RGB Scale:Vector3f=&Multiply_VectorByFloat(
      Vector=<1 1 1>
      Float=&MakeCustomFloatFromBool(
       Bool=$User.bIsFrontendPreview
       True Float=&MakeCustomFloatFromBool(Bool=$User.bIsPainted, True Float=5, False Float=17)
       False Float=&MakeCustomFloatFromBool(Bool=$User.bIsPainted, True Float=2, False Float=10)
      )
     )
     Alpha Scale:NiagaraFloat=1
    )
    Particles.SPEARCOLORSCALE:LinearColor=&VC_DynamicScaleColor(
     Color to Scale:LinearColor=$Particles.SPEARSCOLOR
     RGB Scale:Vector3f=&VectorFromFloat(Value=&MakeCustomFloatFromBool(Bool=$User.bIsPainted, True Float=5, False Float=17))
     Alpha Scale:NiagaraFloat=1
    )
   ) {enabled=false, isSetParameters=true}
   &ParticleState()
   &UpdateMeshOrientation(Facing Interpolation=false, Orientation Method=::NewEnumerator1, Override Quat=false, TransformFacing=false, TransformUp=false) {enabled=false}
   &ScaleMeshSize(Scale Factor=&MaskVectorBySpawnGroup(
     Masked Vector=<1 1 1>
     Passthrough Vector=&MakeCustomVectorFromBool(
      Bool=$User.bIsFrontendPreview
      True Vector=&VectorFromFloat(Value=0.28)
      False Vector=&VectorFromFloat(Value=0.3)
     )
    ))
   &ScaleColor(
    Scale RGB=&MakeCustomVectorFromBool(
     Bool=$User.bIsPainted
     True Vector=&Multiply_VectorByFloat(
      Vector=<1 1 1>
      Float=&MakeCustomFloatFromBool(Bool=$User.bIsFrontendPreview, True Float=1.7, False Float=0.5)
     )
     False Vector=&Multiply_VectorByFloat(
      Vector=&VectorFromFloat(Value=0.7)
      Float=&MakeCustomFloatFromBool(Bool=$User.bIsFrontendPreview, True Float=1.8, False Float=0.75)
     )
    )
    ScaleA=true
    ScaleRGB=true
   )
   /SetVariables(Particles.R1 Color:LinearColor=&VC_DynamicScaleColor(
     Color to Scale:LinearColor=$Particles.Initial.R1 Color
     RGB Scale:Vector3f=&Multiply_VectorByFloat(
      Vector=<1 1 1>
      Float=&MakeCustomFloatFromBool(
       Bool=$User.bIsFrontendPreview
       True Float=&MakeCustomFloatFromBool(Bool=$User.bIsPainted, True Float=10, False Float=8)
       False Float=&MakeCustomFloatFromBool(Bool=$User.bIsPainted, True Float=2.5, False Float=7)
      )
     )
     Alpha Scale:NiagaraFloat=1
    )) {isSetParameters=true}
   &SolveForcesAndVelocity(Clamp Velocity=false, Limit Acceleration=false)
  ]
  rendererClasses [&NiagaraMeshRendererProperties, &NiagaraSpriteRendererProperties]
  renderers [
   {rendererClass=&NiagaraMeshRendererProperties, rendererIndex=0}
   {rendererClass=&NiagaraMeshRendererProperties, rendererIndex=1}
   {rendererClass=&NiagaraSpriteRendererProperties, rendererIndex=2}
  ]
 }
 NE_VC_DriftTrail_Ribbon002 {
  emitterUpdateScript [
   &EmitterState(
    Life Cycle Mode=::NewEnumerator1
    Loop Duration=&MakeCustomFloatFromBool(Bool=$User.bIsFrontendPreview, True Float=1, False Float=0.1)
    UseLoopDelay=false
   )
   &VC_RateQualityHandler(Epic Spawn Rate=40, High Spawn Rate=40, Low Spawn Rate=40, Medium Spawn Rate=40, Use Custom Rate?=true)
   &VC_SpawnRibbon_v1(Number of Ribbons=2, SpawnRate=$Emitter.Spawn Rate)
  ]
  particleSpawnScript [
   &VC_PaintedVariantsOverride(
    Black=<rgba 0.005 0.005 0.005 1>
    BurntSienna=<rgba 0.07 0.0117 0 1>
    Cobalt=<rgba 0 0.015 0.3 1>
    Crimson=<rgba 0.17 0 0.0046 1>
    DefaultColor=<rgba 0 0.1167 0.14 2>
    ForestGreen=<rgba 0.0348 0.18 0.0216 1>
    Grey=<rgba 0.035 0.035 0.035 1>
    Lime=<rgba 0.0556 0.3333 0 1>
    Orange=<rgba 0.3333 0.0222 0 1>
    Pink=&MakeCustomLinearColorFromBool(Bool=$User.bIsFrontendPreview, True Color=<rgba 0.2031 0.0783 0.1563 1>, False Color=<rgba 0.7 0.1667 0.5 1>)
    Purple=<rgba 0.0783 0 0.235 1>
    R1 Black:LinearColor=<rgba 0.005 0.005 0.005 1>
    R1 BurntSienna:LinearColor=<rgba 1.3 0.2221 0.0065 1>
    R1 Cobalt:LinearColor=<rgba 0 0.0393 0.19 1>
    R1 Crimson:LinearColor=<rgba 0.205 0.0041 0.0078 1>
    R1 DefaultColor=<rgba 0 0.111 0.111 0.15>
    R1 ForestGreen:LinearColor=<rgba 0 0.2083 0.0105 1>
    R1 Grey:LinearColor=<rgba 0.034 0.034 0.034 0.2>
    R1 Lime:LinearColor=<rgba 0.0911 0.1823 0 1>
    R1 Orange:LinearColor=<rgba 0.2865 0.0999 0 1>
    R1 Pink:LinearColor=<rgba 2 0.2396 1.3598 1>
    R1 Purple:LinearColor=<rgba 0.0492 0 0.12 0.113>
    R1 Saffron:LinearColor=<rgba 0.668 0.5121 0 1>
    R1 SkyBlue:LinearColor=<rgba 0 0.1456 0.2188 0.2261>
    R1 TitaniumWhite:LinearColor=<rgba 1 1 1 1>
    Saffron=<rgba 0.15 0.1263 0.0075 1>
    SkyBlue=<rgba 0.0135 0.0862 0.45 1>
    TitaniumWhite=<rgba 0.1771 0.1771 0.1771 1>
    Use R1 Painted Overrides=true
   )
   &VC_InitializeRibbon_v1(
    Color=$Particles.Color
    Custom Offset=true
    Enable Skeletal Position=true
    Lifetime=&MakeCustomFloatFromBool(Bool=$User.bIsFrontendPreview, True Float=0.25, False Float=0.26)
    Local Vector Offset=&AddVectorToPosition(
     Position=&MakeCustomVectorFromBool(Bool=$User.bIsFrontendPreview, True Vector=<-25 0 0>, False Vector=<18 0 0>)
     Vector=$User.offset
    )
    Offset Amount=1
    Ribbon Width=21
   )
   &CameraOffset(Camera Offset Amount=21)
   &AddVelocity(
    Velocity=&MakeCustomVectorFromBool(Bool=$User.bIsFrontendPreview, True Vector=<3150 0 0>, False Vector=<3000 0 0>)
    Velocity Speed Scale=-0.6
   ) {module=AddVelocity001}
   &Color(Color=&MakeCustomLinearColorFromBool(
     Bool=$User.bIsFrontendPreview
     True Color=&MakeCustomLinearColorFromBool(Bool=$User.bIsPainted, True Color=$User.VariableForPainted, False Color=<rgba 0 0.5 0.5 1>)
     False Color=&MakeCustomLinearColorFromBool(Bool=$User.bIsPainted, True Color=$User.VariableForPainted, False Color=<rgba 0 0.5 0.5 1>)
    )) {enabled=false}
  ]
  particleUpdateScript [
   &FN_DisableTireFX(Mask Index Attribute:NiagaraInt32=$Particles.SpawnGroup, Mask Index FL=2, Mask Index FR=3, Mask Index RL=0, Mask Index RR=1)
   &ScaleColor(
    Scale RGB=&Multiply_VectorByFloat(
     Vector=<1 1 1>
     Float=&MakeCustomFloatFromBool(
      Bool=$User.bIsFrontendPreview
      True Float=&MakeCustomFloatFromBool(
       Bool=$User.bIsPainted
       True Float=1.7
       False Float=&FloatFromCurve(
        FloatCurve=DataInterface() {propertyValues {Curve {keys [{tangentMode=RCTM_User, time=0, value=0.5}, {tangentMode=RCTM_User, time=1, value=1}]}}}
        Scale Curve=1.7
       )
      )
      False Float=0.5
     )
    )
    ScaleA=true
    ScaleRGB=true
   ) {enabled=false}
   &InheritSourceMovement(
    Applied Position Scale=&Multiply_VectorByFloat(
     Vector=<1 1 1>
     Float=&FloatFromCurve(FloatCurve=DataInterface() {propertyValues {Curve {keys [<at 0.2 1>, {arriveTangent=-0.5052, leaveTangent=-0.5052, time=1, value=0.4}]}}})
    )
    ApplyForce=false
    ApplyPosition=true
    ApplyPositionFalloff=false
   )
   &ParticleState()
   &DynamicMaterialParameters(
    Index 0 Param 1=&MakeCustomFloatFromBool(
     Bool=$User.bIsFrontendPreview
     True Float=&MakeCustomFloatFromBool(Bool=$User.bIsPainted, True Float=&MakeFloatFromInt(Int=$User.CoreBrightness), False Float=4.5)
     False Float=&MakeCustomFloatFromBool(Bool=$User.bIsPainted, True Float=&MakeFloatFromInt(Int=$User.CoreBrightness), False Float=4)
    )
    Module.Index0Param0Write=true
    Module.Index0Param1Write=true
    Module.Index0Param2Write=true
    Module.Index0Param3Write=true
   ) {enabled=false}
   &ScaleRibbonWidth(Ribbon Width Scale=&FloatFromCurve(FloatCurve=DataInterface() {propertyValues {Curve {keys [
         {tangentMode=RCTM_User, time=0, value=1}
         {arriveTangent=-0.0405, leaveTangent=-0.0405, time=0.5, value=1}
         {arriveTangent=-8.0089, leaveTangent=-8.0089, time=1, value=0}
        ]}}}))
   &SolveForcesAndVelocity(Clamp Velocity=false, Limit Acceleration=false)
   &VC_UpdateRibbon_v1(Enable Skeletal Position=true)
   &ScaleColor(
    Scale Alpha=&Multiply_Float(
     A=&MakeCustomFloatFromBool(Bool=$User.bIsPainted, True Float=2, False Float=1)
     B=&FloatFromCurve(FloatCurve=DataInterface() {propertyValues {Curve {keys [<at 0 1>,<at 0.5 1>,<at 1 0>]}}})
    )
    Scale RGB=&Multiply_VectorByFloat(
     Vector=&VectorFromFloat(Value=1.9)
     Float=&MakeCustomFloatFromBool(Bool=$User.bIsFrontendPreview, True Float=3, False Float=0.9)
    )
    ScaleA=true
    ScaleRGB=true
   ) {module=ScaleColor001}
   /SetVariables(Particles.Ribbon2Color:LinearColor=&MultiplyLinearColor(
     Linear Color=$Particles.R1 Color
     Scale Factor=/Niagara/DynamicInputs/Multiply/Multiply_Vector4_ByFloat(
      Vector4:Vector4f=<1 1 1 1>
      Float:NiagaraFloat=&MakeCustomFloatFromBool(Bool=$User.bIsFrontendPreview, True Float=5, False Float=1)
     )
    )) {isSetParameters=true}
  ]
  rendererClasses [&NiagaraRibbonRendererProperties]
  renderers [
   {rendererClass=&NiagaraRibbonRendererProperties}
   {rendererClass=&NiagaraRibbonRendererProperties, rendererIndex=1}
  ]
 }
 NE_VC_DriftTrail_WheelWrap_Parent1 {
  emitterUpdateScript [
   &EmitterState(
    Inactive Response=::NewEnumerator1
    Life Cycle Mode=::NewEnumerator1
    Loop Behavior=::NewEnumerator1
    Loop Duration Mode=::NewEnumerator1
    UseLoopDelay=false
   )
   &SpawnBurst_Instantaneous(Spawn Count=2, Use Spawn Probability=false, UseLoopCountLimit=false)
  ]
  particleSpawnScript [
   &VC_PaintedVariantsOverride(
    Black=<rgba 0.019 0.019 0.019 1>
    BurntSienna=<rgba 0.1875 0.0443 0.0059 1>
    Cobalt=<rgba 0.0057 0.0831 0.3281 1>
    Crimson=<rgba 0.75 0 0.0262 1>
    DefaultColor=<rgba 0.0706 0.3542 0.3542 1>
    ForestGreen=<rgba 0.0524 0.7188 0.0861 1>
    Grey=<rgba 0.12 0.12 0.12 1>
    Lime=<rgba 0.2777 0.7031 0 1>
    Orange=<rgba 0.5 0.1 0.02 1>
    Pink=<rgba 0.5781 0.1987 0.4402 1>
    Purple=<rgba 0.2812 0 0.75 1>
    R1 Black:LinearColor=<rgba 0.088 0.088 0.088 1>
    R1 BurntSienna:LinearColor=<rgba 30 9.2626 3.6979 1>
    R1 Cobalt:LinearColor=<rgba 0.0544 0.7569 1.8 1>
    R1 Crimson:LinearColor=<rgba 0.65 0.1117 0.1263 1>
    R1 DefaultColor=<rgba 0.1777 0.8906 0.8264 1>
    R1 ForestGreen:LinearColor=<rgba 0.2464 1.1 0.2895 1>
    R1 Grey:LinearColor=<rgba 0.4 0.4 0.4 1>
    R1 Lime:LinearColor=<rgba 0.5578 0.9 0.2156 1>
    R1 Orange:LinearColor=<rgba 5 2.0833 0 1>
    R1 Pink:LinearColor=<rgba 40 17.2917 31.7424 1>
    R1 Purple:LinearColor=<rgba 0.5 0.1 0.5 1>
    R1 Saffron:LinearColor=<rgba 3 2.37 0.3 1>
    R1 SkyBlue:LinearColor=<rgba 0.4688 1.1549 1.5 1>
    R1 TitaniumWhite:LinearColor=<rgba 10 10 10 1>
    Saffron=<rgba 0.5 0.388 0.02 1>
    SkyBlue=<rgba 0.08 0.6922 1 1>
    TitaniumWhite=<rgba 0.625 0.625 0.625 1>
    Use R1 Painted Overrides=true
   )
   &InitializeParticle(
    Color=$Particles.Color
    Mesh Scale Mode=::NewEnumerator4
    Sprite Rotation Mode=::NewEnumerator2
    Sprite Size Mode=::NewEnumerator3
    UsePositionOffset=false
    Write Material Random=true
   )
   &PartitionParticles()
   &SkeletalMeshLocation(
    Bone / Socket Index=&Add_Integer(A=$Particles.PartitionParticles.Partition, B=4)
    Bone / Socket Sampling Mode=::NewEnumerator5
    Interpolate Mesh Sampling=false
    MaskSpawnGroup=false
    Mesh Sampling Type=::NewEnumerator6
    Sampled Position Offset=<0 -0.5 0>
    Skeletal Mesh=$System.SkeletalMesh
   )
   &InitialMeshOrientation(Mesh Axis To Orient=<0 0 1>, Mesh Orientation Mode=::NewEnumerator2, Orientation Vector=<0 0 0>, Use Rotation Vector=false)
  ]
  particleUpdateScript [
   &FN_DisableTireFX(
    Mask Index Attribute:NiagaraInt32=$Particles.PartitionParticles.Partition
    Mask Index FL=2
    Mask Index FR=3
    Mask Index RL=0
    Mask Index RR=1
   )
   &VC_WheelWrapScale(Wheel Index Attribute:NiagaraInt32=&Add_Integer(A=$Particles.PartitionParticles.Partition, B=2))
   &ParticleState(Kill Particles When Lifetime Has Elapsed=false)
   &DynamicMaterialParameters(
    Index 0 Param 1=&MakeCustomFloatFromBool(
     Bool=$User.bIsFrontendPreview
     True Float=&MakeCustomFloatFromBool(Bool=$User.bIsPainted, True Float=&MakeFloatFromInt(Int=$User.CoreBrightness), False Float=4.5)
     False Float=&MakeCustomFloatFromBool(Bool=$User.bIsPainted, True Float=&MakeFloatFromInt(Int=$User.CoreBrightness), False Float=4.5)
    )
    Module.Index0Param0Write=true
    Module.Index0Param1Write=true
    Module.Index0Param2Write=true
    Module.Index0Param3Write=true
   ) {module=DynamicMaterialParameters001}
   &UpdateMeshOrientation(Rotation Coordinate Space=::NewEnumerator2, Rotation Vector=<0 1 0>) {enabled=false}
   &SkeletalMeshLocation(
    Bone / Socket Index=&Add_Integer(A=$Particles.PartitionParticles.Partition, B=4)
    Bone / Socket Sampling Mode=::NewEnumerator5
    MaskSpawnGroup=false
    Mesh Sampling Type=::NewEnumerator6
    Skeletal Mesh=$System.SkeletalMesh
   ) {module=SkeletalMeshLocation001}
   &SolveForcesAndVelocity(Clamp Velocity=false, Limit Acceleration=false)
  ]
  rendererClasses [&NiagaraMeshRendererProperties]
  renderers [
   {rendererClass=&NiagaraMeshRendererProperties, rendererIndex=0}
   {rendererClass=&NiagaraMeshRendererProperties, rendererIndex=1}
  ]
 }
}
summary {
 systemName=NS_DriftTrail_LPFire_MP
 emitters [
  {
   emitterName=NE_VC_Core_Outer
   rendererClasses [&NiagaraMeshRendererProperties, &NiagaraSpriteRendererProperties]
  }
  {emitterName=NE_VC_DriftTrail_WheelWrap_Parent1, rendererClasses [&NiagaraMeshRendererProperties]}
  {emitterName=Dots, rendererClasses [&NiagaraSpriteRendererProperties]}
  {emitterName=NE_VC_DriftTrail_Ribbon002, rendererClasses [&NiagaraRibbonRendererProperties]}
 ]
 userVariables [
  User.bIsFrontendPreview {type=NiagaraBool}
  User.bIsPainted {type=NiagaraBool}
  User.ColorVariantID {type=NiagaraInt32}
  User.CoreBrightness {type=NiagaraInt32, defaultValue=2}
  User.DTCarOffset {type=Vector3f, description="", defaultValue {struct=Vector3f, value=<0 0 1>}}
  User.offset {type=Vector3f, description="offset from wheels(set through cosmetic template)", defaultValue {struct=Vector3f, value=<0 0 0>}}
  User.PaintedColor {type=LinearColor, description="", defaultValue {struct=LinearColor, value=<rgba 1 1 1 1>}}
  User.PaintedScalar {type=NiagaraFloat, defaultValue=1, description=""}
  User.SteeringDir {type=NiagaraFloat, defaultValue=0, description=""}
  User.VariableForPainted {type=LinearColor, description="", defaultValue {struct=LinearColor, value=<rgba 1 1 1 1>}}
  User.VariableForPainted2 {type=LinearColor, description="", defaultValue {struct=LinearColor, value=<rgba 1 1 1 1>}}
  User.VehicleFlags {type=NiagaraInt32, defaultValue=806289408}
  User.Wheel_Radius_Array {
   type=NiagaraDataInterfaceArrayFloat
   defaultValue {
    struct=NiagaraExt_VariableValue_DataInterface
    value {
     dataInterface=&self:NiagaraDataInterfaceArrayFloat_0
     dataInterfaceClass=&NiagaraDataInterfaceArrayFloat
    }
   }
  }
  User.Wheel_Width_Array {
   type=NiagaraDataInterfaceArrayFloat
   defaultValue {
    struct=NiagaraExt_VariableValue_DataInterface
    value {
     dataInterface=&self:NiagaraDataInterfaceArrayFloat_1
     dataInterfaceClass=&NiagaraDataInterfaceArrayFloat
    }
   }
  }
  User.WheelSize {type=NiagaraFloat, defaultValue=0, description=""}
 ]
}
